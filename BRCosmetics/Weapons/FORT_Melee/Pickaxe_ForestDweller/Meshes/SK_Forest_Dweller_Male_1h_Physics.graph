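# Physics asset for the one-handed male Forest Dweller pickaxe mesh, giving collision and joint limits to its decorative feathers and ornament.
# A single long box covers the handle bone, three thin boxes cover the dyn_feather bones, and a sphere covers dyn_ornament.
# Each dyn_ body constrains directly to the handle in a star layout, sharing identical moderate swing limits and a tighter twist limit.
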PhysicsAsset /BRCosmetics/Weapons/FORT_Melee/Pickaxe_ForestDweller/Meshes/SK_Forest_Dweller_Male_1h_Physics
bodies {
 Melee_TwoHanded_Handle [{shapeType=Box, extentX=40.4001, extentY=3.5337, extentZ=140.6546, center=<1.1396 -3.5496 30.4381>, rotation=<rot -2.9061 -90 0.0267>}]
 dyn_feather_a [{shapeType=Box, extentX=1.1, extentY=5, extentZ=16, center=<8 0 0>, rotation=<rot 90 0 0>}]
 dyn_feather_b [{shapeType=Box, extentX=1.1, extentY=5, extentZ=16, center=<8 0 0>, rotation=<rot 90 0 0>}]
 dyn_feather_c [{shapeType=Box, extentX=1.1, extentY=5, extentZ=10.8601, center=<5.2589 0 0>, rotation=<rot 90 0 0>}]
 dyn_ornament [{shapeType=Sphere, radius=8.5848, center=<0.1029 -0.0132 0.1623>}]
}
constraints [
 {bone1Name=dyn_feather_a, bone2Name=Melee_TwoHanded_Handle, swing1LimitDegrees=45, swing2LimitDegrees=45, twistLimitDegrees=10}
 {bone1Name=dyn_feather_b, bone2Name=Melee_TwoHanded_Handle, swing1LimitDegrees=45, swing2LimitDegrees=45, twistLimitDegrees=10}
 {bone1Name=dyn_ornament, bone2Name=Melee_TwoHanded_Handle, swing1LimitDegrees=45, swing2LimitDegrees=45, twistLimitDegrees=10}
 {bone1Name=dyn_feather_c, bone2Name=Melee_TwoHanded_Handle, swing1LimitDegrees=45, swing2LimitDegrees=45, twistLimitDegrees=10}
]
